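; Vehicle-mounted machine gun mod weapon that adds equip cues, overheat audio and firing camera shake.
; Heating spawns an attached loop whose "Overheat" parameter follows the normalized heat each tick, fading out on cooled, unequip or end play.
; One-shot sounds mark overheating and failed fire attempts.
; On autonomous proxies, equipping enables mesh ticking and binds camera shake to local ammo changes, because looping fire FX cannot sustain it.
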(event FortWeapon|EventOnEquippedAndAttached
  (bind _returnvalue (Game|GetInstigator))
  ;[1 1 1] On Equipped and Attached
  (GameplayCue|ExecuteGameplayCue _returnvalue (Variables|Default|GetGC_WeaponEquipped))
  (Utilities|IsValid _returnvalue
    (:Is%20Valid
      (if (Utilities|Enum|Equal(Enum) (Networking|GetLocalRole _returnvalue) "ROLE_AutonomousProxy")
        (Components|Tick|SetComponentTickEnabled (Variables|Visual|GetWeaponMesh) true)
        (EventDispatchers|AssignOnLocalAmmoChanged (EventDispatchers|CreateEvent :!event Custom|OnLocalAmmoChanged_Event))))))

(event FortWeapon|EventOnUnEquippedAndDetached
  ;[1 1 1] On Equipped and Attached
  (GameplayCue|ExecuteGameplayCue (Game|GetInstigator) (Variables|Default|GetGC_WeaponUnequip))
  (Components|Tick|SetComponentTickEnabled (Variables|Visual|GetWeaponMesh) false)
  (bind _audiocomponent_overheataudio (Variables|Default|GetAudioComponent_OverheatAudio)
    (:then
      (Audio|Components|Audio|FadeOut _audiocomponent_overheataudio 0.25 0.0 "Linear" :!class /Script/Engine.AudioComponent))
    (:else)))

(event EventEndPlay
  (params (EndPlayReason :type EEndPlayReason%20Enum :default "Destroyed"))
  ;[1 1 1] On Equipped and Attached
  (Components|Tick|SetComponentTickEnabled (Variables|Visual|GetWeaponMesh) false)
  (bind _audiocomponent_overheataudio (Variables|Default|GetAudioComponent_OverheatAudio)
    (:then
      (Audio|Components|Audio|FadeOut _audiocomponent_overheataudio 0.25 0.0 "Linear" :!class /Script/Engine.AudioComponent))
    (:else)))

(event EventTick
  (params (DeltaSeconds :type Float%20%28single-precision%29))
  (CallFunction|UpdateOverheatAudio))

(event Overheating|EventOnStartHeating
  ;[1 1 1] Overheat / Cool Events
  ;;[0 0 0] Heating
  (|CallParentFunction)
  ;[0.1143 1 0.07655] Audio: Overheat Modulation
  (Variables|Default|GetAudioComponent_OverheatAudio
    (:then)
    (:else
      (bind _rootcomponent (Variables|/Script/Engine.Actor|GetRootComponent (Game|GetInstigator))
        (:then
          (bind _returnvalue (Audio|SpawnSoundAttached (Variables|Sounds|GetSound_MountedTurret_OverheatLoop_Cue) _rootcomponent))
          (Variables|Default|SetAudioComponent_OverheatAudio _returnvalue))
        (:else)))))

(event Overheating|EventOnStartOverheated
  ;[1 1 1] Overheat / Cool Events
  ;;[0 0 0] Heating
  (|CallParentFunction)
  ;[1 1 1] Overheated One Shot Audio
  (bind _rootcomponent (Variables|/Script/Engine.Actor|GetRootComponent (Game|GetInstigator))
    (:then
      (Audio|SpawnSoundAttached "/VehicleMods/Mods/MachineGunMod/Audio/Turret_VehMG_Overheat_Cue.Turret_VehMG_Overheat_Cue" _rootcomponent))
    (:else)))

(event Custom|UpdateOverheatAudio
  ;[0.1143 1 0.07655] Audio: Overheat Modulation
  (bind _audiocomponent_overheataudio (Variables|Default|GetAudioComponent_OverheatAudio)
    (:then
      (Audio|Parameter|SetFloatParameter _audiocomponent_overheataudio "Overheat" (/ (Overheat|GetCurrentOverheatValue) (Overheat|GetOverheatingMaxValue)) :!class /Script/AudioExtensions.AudioParameterControllerInterface))
    (:else)))

(event Overheating|EventOnStartCooling
  ;[1 1 1] Overheat / Cool Events
  ;;[0.5521 0.8729 1] Cooling
  (|CallParentFunction))

(event Overheating|EventOnFinishedCooling
  ;[1 1 1] Overheat / Cool Events
  ;;[0.5521 0.8729 1] Cooling
  (|CallParentFunction)
  ;[0.1143 1 0.07655] Audio: Overheat Modulation
  (bind _audiocomponent_overheataudio (Variables|Default|GetAudioComponent_OverheatAudio)
    (:then
      (Audio|Components|Audio|FadeOut _audiocomponent_overheataudio 0.25 0.0 "Linear" :!class /Script/Engine.AudioComponent))
    (:else)))

(event Overheating|EventOnLocalFailedToFireWhenOverheated
  ;[0.1249 1 0.3549] Audio: Fail to Fire when Overheated
  (bind _rootcomponent (Variables|/Script/Engine.Actor|GetRootComponent (Game|GetInstigator))
    (:then
      (Audio|SpawnSoundAttached (Variables|Sounds|GetSound_OnFailToFireWhenOverheated) _rootcomponent))
    (:else)))

(event EventOnPlayWeaponFireFX
  (params (bPersistentFire :type Boolean) (bSecondaryFire :type Boolean))
  (|CallParentFunction bPersistentFire bSecondaryFire))

(event EventOnStopWeaponFireFX
  (|CallParentFunction))

(event Custom|OnLocalAmmoChanged_Event
  (params (LocalCount :type Integer) (LocalRemaining :type Integer))
  ;[1 1 1] On Ammo Changed - Camera Shake (Can't do camera shake on weapon fired FX because it's looping and the camera shake dies
  (Game|Feedback|ClientStartCameraShake :self (Pawn|GetLocalViewingPlayerController (Game|GetInstigator)) :Shake (Variables|Default|GetWeaponFireCameraShake) :Scale 1.0 :PlaySpace "CameraLocal"))

(fn CachePlayerCameraManager
  (params (SeatIndex :type Integer))
  (bind _asfort_athena_vehicle (Utilities|Casting|CastToFortAthenaVehicle (Vehicle|GetHostVehicle)))
  (Utilities|IsValid (Vehicle|GetPawnatSeat _asfort_athena_vehicle SeatIndex)
    (:Is%20Valid
      (Utilities|IsValid (Pawn|GetLocalViewingPlayerController (Vehicle|GetPawnatSeat _asfort_athena_vehicle SeatIndex))
        (:Is%20Valid
          (Utilities|IsValid (Variables|/Script/Engine.PlayerController|GetPlayerCameraManager (Pawn|GetLocalViewingPlayerController (Vehicle|GetPawnatSeat _asfort_athena_vehicle SeatIndex)))
            (:Is%20Valid
              (bind _output_get (Variables|Default|SetPlayer%20Camera%20Manager (Variables|/Script/Engine.PlayerController|GetPlayerCameraManager (Pawn|GetLocalViewingPlayerController (Vehicle|GetPawnatSeat _asfort_athena_vehicle SeatIndex)))))
              (return _output_get))))))))

(fn ConstructionScript)
